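# Material function for the Nexus Pickle port that generates an irregular, time-driven shudder flicker value.
# Offset, speed-scaled time drives several sine waves of differing periods, mixed by sums, products and a lerp, then saturated and sharpened by ShudderPower.
# The main output scales this by Destabilize2 from MPC_Nexus and fades it to zero with PortalWeld; a second output exposes the raw Shudder.
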MaterialFunction /Game/Athena/Environments/Nexus/Pickles/Materials/MF_Pickle_Port_Shudder
graph {nodes [
  n1 = Constant()
  n2 = Multiply(
   A=Add(
    A=Time()
    B=FunctionInput(Preview=@n1) {bUsePreviewValueAsDefault=true, inputName=FlickerTimeOffset, inputType=FunctionInput_Scalar, sortPriority=1}
   )
   B=FunctionInput(Preview=Constant() {r=1}) {bUsePreviewValueAsDefault=true, inputName=FlickerInitialSpeed, inputType=FunctionInput_Scalar}
  ) {constB=80}
  n3 = ConstantBiasScale(None=Sine(None=@n2) {period=0.72})
  n4 = Multiply(A=Sine(None=@n2) {period=0.275}) {constB=0.45}
  n5 = Multiply(A=Multiply(A=Sine(None=@n2) {period=0.555}, B=Sine(None=@n2) {period=0.35})) {constB=0.6}
  n6 = Sine(None=@n2) {period=1.2}
  n7 = ConstantBiasScale(None=Add(A=Add(A=@n4, B=@n5), B=Multiply(A=@n6) {constB=0.5}))
  n8 = Power(Base=Saturate(None=LinearInterpolate(A=@n3, Alpha=@n7) {constA=0.5}), Exp=ScalarParameter() {defaultValue=2, parameterName=ShudderPower})
  n9 = TextureCoordinate()
  Floor(None=Add(A=@n8, B=ComponentMask(None=@n9) {g=true}))
  n10 = CollectionParameter() {collection=/Game/Athena/Environments/Nexus/Materials/FXshader/MPC_Nexus, parameterName=Destabilize2}
  FunctionOutput(None=LinearInterpolate(A=Multiply(A=@n10, B=@n8), Alpha=ScalarParameter() {parameterName=PortalWeld}) {constB=0})
  FunctionOutput(None=@n8) {outputName=Shudder, sortPriority=1}
 ]}
