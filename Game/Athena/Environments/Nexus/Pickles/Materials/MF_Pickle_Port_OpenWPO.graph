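# Material function that drives the opening animation of a Pickle Port portal mesh, with optional welding effects.
# UV-based sphere masking, tangent displacement, sine wobble and shudder build PortalWPO, scaled by an open amount from a parameter or time-driven animation, staggered by vertex-color blue.
# An EnableWelding switch adds weak-point pull WPO, aim-dependent glow and weld fade, and pinches the opening as PortalWeld rises.
# It also outputs camera-distance fade and OpenAmount.
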MaterialFunction /Game/Athena/Environments/Nexus/Pickles/Materials/MF_Pickle_Port_OpenWPO
graph {nodes [
  Divide(A=Time() {bOverride_Period=true, period=3}) {constB=3}
  n1 = ComponentMask(None=TextureCoordinate()) {g=true}
  n2 = Power(Base=@n1, Exp=ScalarParameter() {defaultValue=2.5, group=OpeningWPO, parameterName=SphereCenterPower})
  n3 = FunctionInput(Preview=ScalarParameter() {defaultValue=0.5, group=OpeningWPO, parameterName=SphereCenter}) {bUsePreviewValueAsDefault=true, inputName=SphereCenter, inputType=FunctionInput_Scalar}
  n4 = FunctionInput(Preview=ScalarParameter() {defaultValue=0.5, group=OpeningWPO, parameterName=SphereRadius}) {bUsePreviewValueAsDefault=true, inputName=SphereRadius, inputType=FunctionInput_Scalar}
  n5 = Saturate(None=Power(
    Base=SphereMask(A=@n2, B=@n3, Radius=@n4) {hardnessPercent=0}
    Exp=ScalarParameter() {defaultValue=0.25, group=OpeningWPO, parameterName=SphereOpenPower}
   ) {constExponent=0.6})
  n6 = Time()
  n7 = Sine(None=Add(A=Multiply(A=@n6) {constB=0.5}, B=@n1)) {period=0.42}
  n8 = MaterialFunctionCall!(FlickerInitialSpeed, FlickerTimeOffset=Constant() {r=0.07}) {materialFunction=&/MF_Pickle_Port_Shudder}
  n9 = ComponentMask(None=TextureCoordinate()) {r=true}
  n10 = MaterialFunctionCall() {materialFunction=&TangentBasis}
  n11 = ScalarParameter() {group=OpeningWPO, parameterName=PortalOpenAxisLerp}
  n12 = Add(
   A=Multiply(A=Multiply(A=@n7, B=@n8.Result)) {constB=0.02}
   B=Multiply(A=Subtract(A=@n9) {constB=0.5}, B=LinearInterpolate(A=@n10.X U, B=@n10.X U, Alpha=@n11))
  )
  n13 = ScalarParameter() {defaultValue=550, group=OpeningWPO, parameterName=PortalOpen}
  n14 = OneMinus(None=ComponentMask(None=TextureCoordinate()) {g=true})
  n15 = ScalarParameter() {parameterName=PortalWeld}
  n16 = LinearInterpolate(Alpha=@n15) {constA=0.3, constB=-1.1}
  n17 = LinearInterpolate(Alpha=@n15) {constA=1.2, constB=0.5}
  n18 = Power(Base=NamedRerouteUsage(), Exp=LinearInterpolate(Alpha=VertexColor().B) {constA=0.1, constB=1.1})
  n19 = StaticSwitchParameter(True=Multiply(A=SphereMask(A=@n14, B=@n16, Radius=@n17) {attenuationRadius=1, hardnessPercent=50}, B=@n18) {constA=1}, False=@n18) {parameterName=EnableWelding}
  FunctionOutput(None=Multiply(A=Multiply(A=@n5, B=@n12), B=Multiply(A=@n13, B=@n19))) {outputName=PortalWPO}
  n20 = WorldPosition()
  n21 = ActorPositionWS()
  n22 = Add(
   A=Distance(
    A=StaticSwitchParameter(True=@n20.XYZ, False=@n21) {desc="False uses Actor Position, True uses World Position", group=OpeningWPO, parameterName=UseWorldPositionForDistanceFade}
    B=CameraPositionWS()
   )
   B=ScalarParameter() {defaultValue=-800, group=OpeningWPO, parameterName=CloseDistanceOffset}
  )
  n23 = ScalarParameter() {defaultValue=1500, group=OpeningWPO, parameterName=CloseDistance}
  FunctionOutput(None=Saturate(None=Divide(A=@n22, B=@n23))) {outputName=DistanceFade}
  FunctionOutput(None=@n18) {outputName=OpenAmount}
  n24 = Saturate(None=Ceil(None=@n15))
  n25 = WorldPosition()
  n26 = VectorParameter() {parameterName=WeakPointPos, defaultValue=<rgba 0 0 400 0>}
  n27 = ScalarParameter() {defaultValue=250, parameterName=WeldGlowRadius}
  n28 = MaterialFunctionCall(FlickerInitialSpeed=Constant() {r=4}) {materialFunction=&MF_Pickle_Port_Shudder}
  n29 = Multiply(
   A=VectorParameter() {parameterName=PullDirection, defaultValue=<rgba 1500 0 0 0>}.RGB
   B=LinearInterpolate(A=LinearInterpolate(Alpha=@n28.Shudder) {constA=0.6}, Alpha=ScalarParameter() {defaultValue=1, parameterName=AimSuccess})
  )
  n30 = Multiply(
   A=Power(Base=SphereMask(A=@n25.XYZ, B=@n26.RGB, Radius=@n27) {attenuationRadius=0.25, hardnessPercent=0}) {constExponent=3}
   B=Multiply(A=MaterialFunctionCall(Vector 3=@n29) {materialFunction=/Engine/Functions/Engine_MaterialFunctions02/Utility/VectorLength}.V3 Length) {constB=0.25}
  ) {constB=2}
  n31 = VectorParameter() {parameterName=WeldGlowColorBad, defaultValue=<rgba 5 0.2725 0.1302 0>}
  n32 = VectorParameter() {parameterName=WeldGlowColor, defaultValue=<rgba 0.5336 2.5187 4 0>}
  n33 = ScalarParameter() {defaultValue=1, parameterName=AimSuccess}
  n34 = Multiply(A=@n14, B=@n15)
  n35 = Add(
   A=Multiply(A=Multiply(A=@n24, B=@n30), B=LinearInterpolate(A=@n31.RGB, B=@n32.RGB, Alpha=@n33))
   B=Multiply(A=Multiply(A=@n34, B=@n32.RGB), B=VertexColor().R) {constB=0.5}
  )
  FunctionOutput(None=StaticSwitchParameter(True=Add(A=@n35), False=Constant() {r=1}) {parameterName=EnableWelding}) {outputName=GlowAdd}
  n36 = Power(Base=SphereMask(A=@n25.XYZ, B=@n26.RGB, Radius=ScalarParameter() {defaultValue=400, parameterName=WeldPullRadius}) {hardnessPercent=0})
  FunctionOutput(None=StaticSwitchParameter(True=Multiply(A=@n36, B=@n29), False=Constant()) {parameterName=EnableWelding}) {outputName=WPOAdd, sortPriority=5}
  n37 = TextureCoordinate()
  n38 = Power(Base=@n15) {constExponent=3}
  n39 = Saturate(None=Add(A=ComponentMask(None=@n37) {g=true}, B=LinearInterpolate(Alpha=@n38) {constA=3, constB=-1}))
  FunctionOutput(None=StaticSwitchParameter(True=Multiply(A=@n39, B=@n39), False=Constant() {r=1}) {parameterName=EnableWelding}) {outputName=WeldFade, sortPriority=8}
  n40 = ScalarParameter() {defaultValue=1, group=OpeningWPO, parameterName=AnimationMaxOpenAmount}
  n41 = Time()
  n42 = VectorParameter() {group=OpeningWPO, parameterName=PPOW_AnimParams, channelNames {a=Ease Animation, b=Animation Duration, g=Reverse Animation, r=Game Time}}
  n43 = Saturate(None=Divide(A=Subtract(A=@n41, B=@n42.Game Time), B=@n42.Animation Duration))
  n44 = LinearInterpolate(A=@n43, B=SmoothStep(Value=@n43), Alpha=@n42.Ease Animation)
  n45 = LinearInterpolate(A=@n44, B=OneMinus(None=@n44), Alpha=@n42.Reverse Animation)
  NamedRerouteDeclaration(None=StaticSwitchParameter(True=Multiply(A=@n40, B=@n45), False=ScalarParameter() {group=OpeningWPO, parameterName=OpenAmount}) {desc=Change animation from parameter driven to material time driven, group=OpeningWPO, parameterName=UseMaterialAnimation}) {name=Open Amount, nodeColor=<rgba 1 0 0.8437 1>}
 ]}
