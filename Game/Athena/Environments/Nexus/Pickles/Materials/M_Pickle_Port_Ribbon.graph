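# Translucent unlit, two-sided ribbon material for the Pickle Port in the Nexus environment, usable on spline meshes and Niagara ribbons.
# Panning star-line layers, distorted by chromatic caustics, combine with a sine-wave two-color background.
# MF_Pickle_Port_OpenWPO, fed by vertex-color red, drives width-based tiling, wave phase and glow.
# Opacity uses edge and end masks, plus view-angle, camera and depth fades, replaced by a constant on forward and mobile.
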Material /Game/Athena/Environments/Nexus/Pickles/Materials/M_Pickle_Port_Ribbon
graph {
 properties {
  bUsedWithNiagaraRibbons=true
  bUsedWithSplineMeshes=true
  blendMode=BLEND_Translucent
  pixelDepthOffsetMode=PDOM_Legacy
  shadingModel=MSM_Unlit
  twoSided=true
  usedShadingModels=Unlit
 }
 nodes [
  n1 = VectorParameter() {parameterName=EmissiveMult, defaultValue=<rgba 1.0732 1.3311 1.5 0>}
  n2 = VectorParameter() {parameterName=Tile1, defaultValue=<rgba 0.25 10 0 0>}
  n3 = ScalarParameter() {defaultValue=200, parameterName=SplineWidth}
  n4 = VertexColor()
  n5 = MaterialFunctionCall!(
   SphereRadius=LinearInterpolate(
    A=ScalarParameter() {defaultValue=0.6, parameterName=SphereRadiusOuter}
    B=ScalarParameter() {defaultValue=0.5, parameterName=SphereRadiusInner}
    Alpha=@n4.R
   )
   SphereCenter=LinearInterpolate(
    A=ScalarParameter() {defaultValue=0.6, parameterName=SphereCenterOuter}
    B=ScalarParameter() {defaultValue=0.5, parameterName=SphereCenterInner}
    Alpha=@n4.R
   )
  ) {materialFunction=&/MF_Pickle_Port_OpenWPO}
  n6 = LinearInterpolate(Alpha=@n5.DistanceFade) {constA=0.5}
  n7 = VertexColor()
  n8 = Multiply(A=ScalarParameter() {defaultValue=0.02, parameterName=SplineWidthTileMult}, B=Multiply(A=Multiply(A=@n3, B=@n6), B=@n7.B)) {constB=0.1}
  n9 = Multiply(A=TextureCoordinate() {coordinateIndex=1}, B=AppendVector(A=Multiply(A=@n2.R, B=@n8), B=@n2.G))
  n10 = VectorParameter() {parameterName=Tile2, defaultValue=<rgba 1 0.15 0 0>}
  n11 = AppendVector(A=@n10.R, B=@n10.R)
  n12 = VertexColor()
  n13 = Time()
  n14 = Add(A=@n12.G, B=Multiply(A=Multiply(A=@n12.B, B=@n13), B=ScalarParameter() {defaultValue=-0.5, parameterName=PanSpeed}))
  n15 = VectorParameter() {parameterName=PanDirection2, defaultValue=<rgba 0.04 0.1 0 0>}
  n16 = TextureSample(UVs=Panner(Coordinate=Multiply(A=@n9, B=@n11), Time=@n14, Speed=AppendVector(A=@n15.R, B=@n15.R)) {speedX=1, speedY=1}) {texture=/Game/Effects/Fort_Effects/Textures/Tile/T_Caustics_Chromatic}
  n17 = Multiply(A=ComponentMask(None=@n16.RGB) {g=true, r=true}) {constB=0.05}
  n18 = VectorParameter() {parameterName=PanDirection1, defaultValue=<rgba 0.08 1 0 0>}
  n19 = AppendVector(A=@n18.R, B=@n18.R)
  n20 = TextureSample(UVs=Panner(Coordinate=Add(A=@n9, B=@n17), Time=@n14, Speed=@n19) {speedX=1, speedY=1}) {texture=/Game/Athena/Environments/Nexus/Textures/T_ZeroPoint_Line_Stars}
  n21 = Multiply(A=@n9) {constB=2}
  n22 = VectorParameter() {parameterName=PanDirection1Mult, defaultValue=<rgba -1.7 2.4 0 1>}
  n23 = AppendVector(A=@n22.R, B=@n22.R)
  n24 = TextureSample(UVs=Panner(Coordinate=Subtract(A=@n21, B=@n17), Time=@n14, Speed=Multiply(A=@n19, B=@n23)) {speedX=1, speedY=1}) {texture=&T_ZeroPoint_Line_Stars}
  n25 = Multiply(A=@n24.RGB, B=VectorParameter() {parameterName=SecondaryTint, defaultValue=<rgba 0.4187 0.0375 0.8 1>}.RGB) {constB=0.5}
  n26 = VectorParameter() {parameterName=BackgroundColor1, defaultValue=<rgba 0.0594 0.2785 0.3 0>}
  n27 = VectorParameter() {parameterName=BackgroundColor2, defaultValue=<rgba 0.0584 0.0394 0.2 0>}
  n28 = Multiply(A=@n7.G) {constB=2}
  n29 = TextureCoordinate() {coordinateIndex=1}
  n30 = Time()
  n31 = ScalarParameter() {defaultValue=-0.08, parameterName=WaveSpeed}
  n32 = Add(
   A=Divide(A=ComponentMask(None=@n29) {g=true}, B=ScalarParameter() {defaultValue=0.75, parameterName=WaveLength})
   B=Add(A=Multiply(A=@n30, B=@n31), B=@n5.OpenAmount)
  )
  n33 = ConstantBiasScale(None=Sine(None=Add(A=@n28, B=@n32)))
  n34 = TextureCoordinate() {coordinateIndex=1}
  n35 = Constant() {r=0.5}
  n36 = Multiply(A=SphereMask(A=ComponentMask(None=@n34) {r=true}, B=@n35, Radius=@n35) {hardnessPercent=0})
  n37 = Multiply(
   A=MaterialFunctionCall(In=@n36, Contrast=Constant() {r=0.25}) {materialFunction=&CheapContrast}.Result
   B=VertexColor().B
  )
  n38 = Add(
   A=Add(A=Add(A=@n20.RGB, B=@n25), B=Multiply(A=@n16.RGB) {constB=0.5})
   B=Multiply(A=LinearInterpolate(A=@n26.RGB, B=@n27.RGB, Alpha=@n33), B=@n37)
  )
  MaterialFunctionCall(Emissive=Multiply(A=Multiply(A=@n1.RGB, B=@n38), B=@n5.GlowAdd)) {materialFunction=&MF_FX_EmissiveAdjustments}
  n39 = Add(A=@n24.B, B=@n16.B)
  n40 = Multiply(A=Add(A=Add(A=@n20.B, B=@n39), B=Multiply(A=@n37) {constB=0.25}) {constB=0.5}, B=@n36)
  n41 = Constant() {r=0.5}
  n42 = SphereMask(
   A=Power(Base=ComponentMask(None=@n34) {g=true})
   B=@n41
   Radius=@n41
   Hardness=ScalarParameter() {defaultValue=0.5, parameterName=EndFadeHardness}
  ) {hardnessPercent=0.25}
  n43 = MaterialFunctionCall("FadeSensitivity (optional)"=ScalarParameter() {defaultValue=1, parameterName=AngleFade}) {materialFunction=&FadeBasedOnViewAngle}
  n44 = ScalarParameter() {defaultValue=1200, parameterName=NearFadeDIst}
  n45 = ScalarParameter() {defaultValue=100, parameterName=DepthFade}
  n46 = Multiply(
   A=Saturate(None=Multiply(A=@n43.Result) {constB=2})
   B=Multiply(
    A=MaterialFunctionCall(Distance=@n44) {materialFunction=&Near_Camera_Fade}.Result
    B=DepthFade(FadeDistance=@n45)
   )
  )
  n47 = Constant() {r=0.7}
  Multiply(
   A=Multiply(A=Multiply(A=@n40, B=@n42), B=ShadingPathSwitch(Default=@n46, Forward=@n47, Mobile=@n47))
   B=ScalarParameter() {defaultValue=1, parameterName=OpacityMult}
  )
 ]
}
