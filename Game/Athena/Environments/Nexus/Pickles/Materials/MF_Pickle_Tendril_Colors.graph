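# Material function that supplies a position-based gradient color and alpha for Pickle tendril materials in the Nexus environment.
# A soft-edged SphereMask on world position, centered on the ZeroPoiintPos parameter with a GradientRadius falloff, drives two lerps.
# These blend the EmissiveMultEnd and EmissiveMultStart vector parameters, outputting RGB and alpha separately, with the Start values at the center and the End values beyond the radius.
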MaterialFunction /Game/Athena/Environments/Nexus/Pickles/Materials/MF_Pickle_Tendril_Colors
graph {nodes [
  n1 = VectorParameter() {parameterName=EmissiveMultEnd, defaultValue=<rgba 1.6094 0.8782 3 1>}
  n2 = VectorParameter() {parameterName=EmissiveMultStart, defaultValue=<rgba 2 0.6042 0.9818 0.5>}
  n3 = SphereMask(
   A=WorldPosition().XYZ
   B=VectorParameter() {group=TendrilFX, parameterName=ZeroPoiintPos, defaultValue=<rgba 6975 5908 8064 2000>}.RGB
   Radius=ScalarParameter() {defaultValue=40000, parameterName=GradientRadius}
  ) {hardnessPercent=0}
  FunctionOutput(None=LinearInterpolate(A=@n1.RGB, B=@n2.RGB, Alpha=@n3))
  FunctionOutput(None=LinearInterpolate(A=@n1.A, B=@n2.A, Alpha=@n3)) {outputName=Alpha, sortPriority=1}
 ]}
