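# Material function for the Nexus environment that pixelates input UVs when the Skirmish static switch is on, and otherwise passes them through unchanged.
# Multiplies the UVs by a resolution, rounds them and divides back, snapping them to a grid.
# Time-driven sawtooth waves at differing rates are mixed into the Resolution input at reduced weight, so the pixel grid jitters continuously around that value.
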MaterialFunction /Game/Athena/Environments/Nexus/Pickles/Materials/MF_Skirmish_Pixelate
graph {nodes [
  n1 = FunctionInput(Preview=TextureCoordinate()) {inputName=UVs, inputType=FunctionInput_Vector2}
  n2 = Time()
  n3 = LinearInterpolate(A=Frac(None=Multiply(A=@n2) {constB=2.7}), B=Frac(None=Multiply(A=@n2) {constB=2}))
  n4 = Multiply(A=Frac(None=Multiply(A=@n2) {constB=1.1})) {constB=-0.5}
  n5 = FunctionInput() {bUsePreviewValueAsDefault=true, inputName=Resolution, inputType=FunctionInput_Scalar}
  n6 = LinearInterpolate(A=Add(A=Add(A=@n3, B=@n4), B=@n5), B=@n5) {constAlpha=0.6}
  n7 = Round(None=Multiply(A=@n1, B=@n6))
  FunctionOutput(None=StaticSwitchParameter(True=Divide(A=@n7, B=@n6), False=@n1) {parameterName=Skirmish})
 ]}
